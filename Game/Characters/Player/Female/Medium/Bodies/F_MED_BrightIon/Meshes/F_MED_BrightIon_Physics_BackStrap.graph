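# Physics asset for the dangling back strap on the medium female player body F_MED_BrightIon, for secondary motion.
# Capsule bodies on the dyn_back_strap bones form a constrained chain from the MSTR bone through four segments.
# Angular limits are tight at the root, with twist freedom increasing toward the tip.
# Unconstrained capsules on the pelvis, spine_02 and both thighs approximate the body around the strap, apparently as colliders.
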PhysicsAsset /Game/Characters/Player/Female/Medium/Bodies/F_MED_BrightIon/Meshes/F_MED_BrightIon_Physics_BackStrap
bodies {
 dyn_back_strap_01 [{shapeType=Capsule, length=3, radius=1.5, shapeName=dyn_back_strap_01_capsule, center=<2.2439 0 0>, rotation=<rot 90 0 0>}]
 dyn_back_strap_02 [{shapeType=Capsule, length=9.5, radius=2, shapeName=dyn_back_strap_02_capsule, center=<5.6379 0 0>, rotation=<rot 90 0 0>}]
 dyn_back_strap_03 [{shapeType=Capsule, length=11, radius=2, shapeName=dyn_back_strap_03_capsule, center=<6.5716 0 0>, rotation=<rot 90 0 0>}]
 dyn_back_strap_04 [{shapeType=Capsule, length=12, radius=2, shapeName=dyn_back_strap_04_capsule, center=<6.6479 0 0>, rotation=<rot 90 0 0>}]
 dyn_back_strap_MSTR [{shapeType=Capsule, length=2, radius=0.5, shapeName=dyn_back_strap_MSTR_capsule}]
 pelvis [{shapeType=Capsule, length=7.5, radius=11.5, shapeName=pelvis_capsule, center=<-5.2147 -1.4348 0>}]
 spine_02 [{shapeType=Capsule, length=8, radius=8.5, shapeName=spine_02_capsule, center=<-4.2359 1.0441 0>}]
 thigh_l [{
   shapeType=Capsule
   length=19.5
   radius=8
   shapeName=thigh_l_capsule
   center=<-16.4444 -2.0965 1.204>
   rotation=<rot 84.6272 169.3766 179.9999>
  }]
 thigh_r [{
   shapeType=Capsule
   length=21.5
   radius=8
   shapeName=thigh_r_capsule
   center=<14.9228 2.2043 -1.1646>
   rotation=<rot 83.986 179.999 -167.0563>
  }]
}
constraints [
 {bone1Name=dyn_back_strap_01, bone2Name=dyn_back_strap_MSTR, swing1LimitDegrees=5, swing2LimitDegrees=5, twistLimitDegrees=2}
 {bone1Name=dyn_back_strap_02, bone2Name=dyn_back_strap_01, swing1LimitDegrees=20, swing2LimitDegrees=20, twistLimitDegrees=10}
 {bone1Name=dyn_back_strap_03, bone2Name=dyn_back_strap_02, swing1LimitDegrees=20, swing2LimitDegrees=20, twistLimitDegrees=20}
 {bone1Name=dyn_back_strap_04, bone2Name=dyn_back_strap_03, swing1LimitDegrees=20, swing2LimitDegrees=20, twistLimitDegrees=30}
]
