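# Physics asset for the front belt and jacket corner of the medium female BrightIon player body.
# Capsule bodies form three short dynamic bone chains (left belt, right belt, jacket corner), each constrained with angular limits back to its MSTR root.
# The left belt swings most freely, while the right belt and jacket corner are tighter, the corner barely twisting.
# Unconstrained pelvis, lower spine and thigh capsules supply body volumes alongside the chains.
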PhysicsAsset /Game/Characters/Player/Female/Medium/Bodies/F_MED_BrightIon/Meshes/F_MED_BrightIon_Physics_FrontBelt
bodies {
 dyn_belt_l_01 [{shapeType=Capsule, length=6, radius=2, shapeName=dyn_belt_l_01_capsule, center=<3.6582 0 0>, rotation=<rot 90 0 0>}]
 dyn_belt_l_02 [{shapeType=Capsule, length=3, radius=2.5, shapeName=dyn_belt_l_02_capsule, center=<2.7994 0 0>, rotation=<rot 90 0 0>}]
 dyn_belt_l_03 [{shapeType=Capsule, length=5, radius=2.5, shapeName=dyn_belt_l_03_capsule, center=<3.6514 0 0>, rotation=<rot 90 0 0>}]
 dyn_belt_l_MSTR [{shapeType=Capsule, length=2, radius=0.5, shapeName=dyn_belt_l_MSTR_capsule}]
 dyn_belt_r_01 [{shapeType=Capsule, length=3, radius=1.5, shapeName=dyn_belt_r_01_capsule, center=<2.4252 0 0.55>}]
 dyn_belt_r_02 [{shapeType=Capsule, length=3, radius=1.5, shapeName=dyn_belt_r_02_capsule, center=<2.9447 0 0>}]
 dyn_belt_r_MSTR [{shapeType=Capsule, length=2, radius=0.5, shapeName=dyn_belt_r_MSTR_capsule}]
 dyn_jacket_corner_01 [{shapeType=Capsule, length=4, radius=2, shapeName=dyn_jacket_corner_01_capsule, center=<1.7 0.5 1>, rotation=<rot 90 180 -90>}]
 dyn_jacket_corner_02 [{shapeType=Capsule, length=1.5, radius=1.5, shapeName=dyn_jacket_corner_02_capsule, center=<3 0.3 0.5>, rotation=<rot 90 153.4349 -116.5651>}]
 dyn_jacket_corner_MSTR [{shapeType=Capsule, length=2, radius=0.5, shapeName=dyn_jacket_corner_MSTR_capsule}]
 pelvis [{shapeType=Capsule, length=12, radius=10.5, shapeName=pelvis_capsule, center=<-3.3148 -2.3636 0>}]
 spine_01 [{shapeType=Capsule, length=8.5, radius=11.5, shapeName=spine_01_capsule, center=<0 -1.9561 0>}]
 spine_02 [{shapeType=Capsule, length=5.5, radius=7.5, shapeName=spine_02_capsule, center=<0 0.9702 0>}]
 thigh_l [{
   shapeType=Capsule
   length=30
   radius=8.5
   shapeName=thigh_l_capsule
   center=<-17.9638 0.4467 0.6534>
   rotation=<rot 89.8828 -163.3931 -161.0234>
  }]
 thigh_r [{shapeType=Capsule, length=30, radius=8.5, shapeName=thigh_r_capsule, center=<18.0474 0.4998 -1.0432>, rotation=<rot 84.3429 180 180>}]
}
constraints [
 {bone1Name=dyn_belt_r_02, bone2Name=dyn_belt_r_01, swing1LimitDegrees=20, swing2LimitDegrees=10, twistLimitDegrees=10}
 {bone1Name=dyn_jacket_corner_02, bone2Name=dyn_jacket_corner_01, swing1LimitDegrees=5, swing2LimitDegrees=30, twistLimitDegrees=1}
 {bone1Name=dyn_jacket_corner_01, bone2Name=dyn_jacket_corner_MSTR, swing1LimitDegrees=5, swing2LimitDegrees=40, twistLimitDegrees=1}
 {bone1Name=dyn_belt_l_01, bone2Name=dyn_belt_l_MSTR, swing1LimitDegrees=40, swing2LimitDegrees=30, twistLimitDegrees=20}
 {bone1Name=dyn_belt_l_02, bone2Name=dyn_belt_l_01, swing1LimitDegrees=40, swing2LimitDegrees=40, twistLimitDegrees=30}
 {bone1Name=dyn_belt_l_03, bone2Name=dyn_belt_l_02, swing1LimitDegrees=45, swing2LimitDegrees=45, twistLimitDegrees=30}
 {bone1Name=dyn_belt_r_01, bone2Name=dyn_belt_r_MSTR, swing1LimitDegrees=15, swing2LimitDegrees=20, twistLimitDegrees=10}
]
