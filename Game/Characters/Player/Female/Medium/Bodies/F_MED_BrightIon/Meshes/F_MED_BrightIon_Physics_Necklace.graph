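# Physics asset for the dangling necklace on the F_MED_BrightIon medium female player character, covering only the necklace and upper-spine bones.
# A three-link capsule chain hangs from dyn_necklace_MSTR, with narrow swing and twist limits between links to keep motion restrained.
# The master bone joins spine_04 with free swing and twist.
# Large capsules on spine_04 and spine_05 form the torso volume, and spine_05 has no constraint.
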PhysicsAsset /Game/Characters/Player/Female/Medium/Bodies/F_MED_BrightIon/Meshes/F_MED_BrightIon_Physics_Necklace
bodies {
 dyn_necklace_01 [{shapeType=Capsule, length=11, radius=6.5, shapeName=dyn_necklace_01_capsule, center=<4.9595 -6.2064 0>}]
 dyn_necklace_02 [{shapeType=Capsule, length=5, radius=8.5, shapeName=dyn_necklace_02_capsule, center=<3.6564 -7.9453 0>}]
 dyn_necklace_03 [{shapeType=Capsule, length=2, radius=6, shapeName=dyn_necklace_03_capsule, center=<3.1131 -5.1094 0>}]
 dyn_necklace_MSTR [{shapeType=Capsule, length=2, radius=0.5, shapeName=dyn_necklace_MSTR_capsule}]
 spine_04 [{shapeType=Capsule, length=20.5, radius=21, shapeName=spine_04_capsule, center=<-4.3161 -8.1364 0>}]
 spine_05 [{shapeType=Capsule, length=22, radius=8.1, shapeName=spine_05_capsule}]
}
constraints [
 {bone1Name=dyn_necklace_02, bone2Name=dyn_necklace_01, swing1LimitDegrees=40, swing2LimitDegrees=20, twistLimitDegrees=5}
 {bone1Name=dyn_necklace_03, bone2Name=dyn_necklace_02, swing1LimitDegrees=10, swing2LimitDegrees=10, twistLimitDegrees=5}
 {bone1Name=dyn_necklace_01, bone2Name=dyn_necklace_MSTR, swing1LimitDegrees=10, swing2LimitDegrees=20, twistLimitDegrees=1}
 {
  bone1Name=dyn_necklace_MSTR
  bone2Name=spine_04
  swing1LimitDegrees=45
  swing1Motion=Free
  swing2LimitDegrees=45
  swing2Motion=Free
  twistLimitDegrees=45
  twistMotion=Free
 }
]
