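# Post-process style AnimBP for the F_MED_BrightIon player body, layering Control Rig and accessory secondary motion onto the attached parent's pose.
# Back strap switches between simulation and constrained or offset poses based on backbling, cape, skydive and front-end state.
# Chained RigidBody sims drive front belt and necklace, with belt and necklace correctives weighted by bone deltas.
# EventGraph updates those alphas, gravity and velocity scales, and feeds WPO cloth-wind material parameters.
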AnimBlueprint /Game/Characters/Player/Female/Medium/Bodies/F_MED_BrightIon/Meshes/F_MED_BrightIon_AnimBP
target_skeleton=&/F_MED_BrightIon_Skeleton
graphs [
 AnimGraph {nodes [
   Output Pose(Result=Use cached pose 'Final_Pose'()) {layerGroup=None}
   n1 = RigidBody!(
    ComponentPose=Local To Component(LocalPose=Use cached pose 'PreSim_Pose'())
    Alpha
    OverrideWorldGravity:Vector="0,0,-980"
    ComponentLinearAccScale:Vector="0.3,0.3,0.5"
    ComponentLinearVelScale:Vector="0.2,0.2,0.3"
   ) -> (Pose) {
    bClampLinearTranslationLimitToRefPose=true
    bOverrideWorldGravity=true
    overridePhysicsAsset=&/F_MED_BrightIon_Physics_BackStrap
    simulationSpace=BaseBoneSpace
    alphaBoolBlend {blendOption=Linear}
    baseBoneRef {boneName=root}
    componentAppliedLinearAccClamp=<3000 3000 3000>
    componentLinearAccScale=<0.3 0.3 0.5>
    componentLinearVelScale=<0.2 0.2 0.3>
    overrideWorldGravity {z=-980}
   }
   n2 = Component To Local(ComponentPose=@n1)
   n3 = Use cached pose 'WithBack_Pose'()
   n4 = Get WearingBack() -> (WearingBack:bool)
   n5 = RigidBody(
    ComponentPose=Local To Component(LocalPose=Blend Poses by bool(BlendPose_0=@n3, BlendPose_1=@n2, bActiveValue=@n4) {blendType=Linear})
    OverrideWorldGravity="0,0,-980"
    ComponentLinearAccScale="0.3,0.3,0.5"
    ComponentLinearVelScale="0.2,0.2,0.3"
   ) {
    bClampLinearTranslationLimitToRefPose=true
    bOverrideWorldGravity=true
    overridePhysicsAsset=&/F_MED_BrightIon_Physics_FrontBelt
    simulationSpace=BaseBoneSpace
    alphaBoolBlend {blendOption=Linear}
    baseBoneRef {boneName=root}
    componentAppliedLinearAccClamp=<3000 3000 3000>
    componentLinearAccScale=<0.3 0.3 0.5>
    componentLinearVelScale=<0.2 0.2 0.3>
    overrideWorldGravity {z=-980}
   }
   n6 = RigidBody(ComponentPose=@n5, OverrideWorldGravity="0,0,-980", ComponentLinearAccScale="0.3,0.3,0.5", ComponentLinearVelScale="0.2,0.2,0.3") {
    bClampLinearTranslationLimitToRefPose=true
    bOverrideWorldGravity=true
    overridePhysicsAsset=&/F_MED_BrightIon_Physics_Necklace
    simulationSpace=BaseBoneSpace
    alphaBoolBlend {blendOption=Linear}
    baseBoneRef {boneName=spine_05}
    componentAppliedLinearAccClamp=<6000 6000 6000>
    componentLinearAccScale=<0.3 0.3 0.5>
    componentLinearVelScale=<0.2 0.2 0.3>
    overrideWorldGravity {z=-980}
   }
   n7 = "Transform (Modify) Bone - Bone: dyn_necklace_02"!(ComponentPose=@n6, Alpha=Get Alpha_Necklace_L() -> (Alpha_Necklace_L:double), Translation="0,0,0", Rotation="0,0,-10") -> (Pose) {
    rotationMode=BMM_Additive
    translationMode=BMM_Additive
    alphaBoolBlend {blendOption=Linear}
    boneToModify {boneName=dyn_necklace_02}
    rotation {roll=-10}
   }
   n8 = "Transform (Modify) Bone - Bone: dyn_necklace_02"(ComponentPose=@n7, Alpha=Get Alpha_Necklace_R() -> (Alpha_Necklace_R:double), Translation="0,0,0", Rotation="0,0,-10") {
    rotationMode=BMM_Additive
    translationMode=BMM_Additive
    alphaBoolBlend {blendOption=Linear}
    boneToModify {boneName=dyn_necklace_02}
    rotation {roll=-10}
   }
   Save cached pose 'Final_Pose'(Pose=Component To Local(ComponentPose=@n8)) {cacheName=Final_Pose}
   n9 = NOT Boolean(A=Get bIsInFrontEnd())
   n10 = "Transform (Modify) Bone - Bone: dyn_back_strap_01"!(
    ComponentPose=Local To Component(LocalPose=Use cached pose 'PreSim_Pose'())
    bAlphaBoolEnabled:bool=@n9
    Translation="-0.000,2.444,-0.000"
    Rotation="0.000,0.000,-27.913"
   ) -> (Pose) {
    alphaInputType=Bool
    rotationMode=BMM_Additive
    scaleMode=BMM_Additive
    scaleSpace=BCS_BoneSpace
    translationMode=BMM_Additive
    alphaBoolBlend {blendOption=Linear}
    boneToModify {boneName=dyn_back_strap_01}
    rotation {roll=-27.913}
    translation {y=2.444}
   }
   n11 = "Constraint - dyn_back_strap_02 (P)"(ComponentPose=@n10, Alpha=1, ConstraintWeights_0=1) {
    alphaBoolBlend {blendOption=Linear}
    boneToModify {boneName=dyn_back_strap_02}
    constraintSetup [{targetBone {boneName=spine_02}}]
    constraintWeights [1]
   }
   n12 = "Transform (Modify) Bone - Bone: dyn_back_strap_02"!(ComponentPose=@n11, bAlphaBoolEnabled:bool=@n9, Translation="0.000,4.766,0.000", Rotation="-7.564,9.224,-8.705") -> (Pose) {
    alphaInputType=Bool
    rotationMode=BMM_Additive
    scaleSpace=BCS_BoneSpace
    translationMode=BMM_Additive
    alphaBoolBlend {blendOption=Linear}
    boneToModify {boneName=dyn_back_strap_02}
    rotation=<rot -7.564 9.224 -8.705>
    translation {y=4.766}
   }
   n13 = "Constraint - dyn_back_strap_03 (P)"(ComponentPose=@n12, Alpha=1, ConstraintWeights_0=1) {
    alphaBoolBlend {blendOption=Linear}
    boneToModify {boneName=dyn_back_strap_03}
    constraintSetup [{targetBone {boneName=pelvis}}]
    constraintWeights [1]
   }
   n14 = "Transform (Modify) Bone - Bone: dyn_back_strap_03"!(ComponentPose=@n13, bAlphaBoolEnabled:bool=@n9, Translation="0.000,0.417,3.470", Rotation="0,0,0") -> (Pose) {
    alphaInputType=Bool
    rotationMode=BMM_Additive
    scaleMode=BMM_Additive
    scaleSpace=BCS_BoneSpace
    translationMode=BMM_Additive
    alphaBoolBlend {blendOption=Linear}
    boneToModify {boneName=dyn_back_strap_03}
    translation {y=0.417, z=3.47}
   }
   n15 = "Constraint - dyn_back_strap_04 (PPP)"!(ComponentPose=@n14, Alpha=1, ConstraintWeights_0=0.125, ConstraintWeights_1:float=0.65, ConstraintWeights_2:float=0.225) -> (Pose) {
    alphaBoolBlend {blendOption=Linear}
    boneToModify {boneName=dyn_back_strap_04}
    constraintSetup [{targetBone {boneName=deform_glute_r}}, {targetBone {boneName=pelvis}}, {targetBone {boneName=thigh_twist_01_r}}]
    constraintWeights [1,1,1]
   }
   n16 = "Transform (Modify) Bone - Bone: dyn_back_strap_04"!(ComponentPose=@n15, Alpha=1, Translation="0,0,0", Rotation="0.000,0.000,-5.544") -> (Pose) {
    rotationMode=BMM_Additive
    scaleSpace=BCS_BoneSpace
    translationMode=BMM_Additive
    alphaBoolBlend {blendOption=Linear}
    boneToModify {boneName=dyn_back_strap_04}
    rotation {roll=-5.544}
   }
   n17 = "Transform (Modify) Bone - Bone: dyn_back_strap_02"!(
    ComponentPose=Local To Component(LocalPose=Use cached pose 'Back_Pose'())
    Alpha:float=1
    Translation="0.107,4.757,-1.770"
    Rotation="-0.000,0.000,18.743"
   ) -> (Pose) {
    rotationMode=BMM_Additive
    scaleMode=BMM_Additive
    scaleSpace=BCS_BoneSpace
    translationMode=BMM_Additive
    translationSpace=BCS_BoneSpace
    alphaBoolBlend {blendOption=Linear}
    boneToModify {boneName=dyn_back_strap_02}
    rotation {roll=18.743}
    translation=<0.107 4.757 -1.77>
   }
   n18 = "Transform (Modify) Bone - Bone: dyn_back_strap_02"(ComponentPose=Local To Component(LocalPose=Use cached pose 'Back_Pose'()), Alpha=1, Translation="-0.5,3.5,-1.3", Rotation="0,0,11") {
    rotationMode=BMM_Additive
    scaleMode=BMM_Additive
    scaleSpace=BCS_BoneSpace
    translationMode=BMM_Additive
    translationSpace=BCS_BoneSpace
    alphaBoolBlend {blendOption=Linear}
    boneToModify {boneName=dyn_back_strap_02}
    rotation {roll=11}
    translation=<-0.5 3.5 -1.3>
   }
   n19 = Blend Poses by bool(
    BlendPose_0=Component To Local(ComponentPose=@n17)
    BlendPose_1=Component To Local(ComponentPose=@n18)
    BlendTime_0=1
    BlendTime_1=0.5
    bActiveValue=Get bIsSkydiveDiveMode()
   ) {blendType=Linear}
   n20 = Use cached pose 'Back_Pose'()
   n21 = Get bIsSkyDiving()
   n22 = RigidBody(
    ComponentPose=Local To Component(LocalPose=Blend Poses by bool(BlendPose_0=@n19, BlendPose_1=@n20, bActiveValue=@n21) {blendType=Linear})
    OverrideWorldGravity="0,0,-980"
    ComponentLinearAccScale="0.3,0.3,0.5"
    ComponentLinearVelScale="0.2,0.2,0.3"
   ) {
    bClampLinearTranslationLimitToRefPose=true
    bOverrideWorldGravity=true
    overridePhysicsAsset=&/F_MED_BrightIon_Physics_BackStrap_WithBack
    simulationSpace=BaseBoneSpace
    alphaBoolBlend {blendOption=Linear}
    baseBoneRef {boneName=root}
    componentAppliedLinearAccClamp=<3000 3000 3000>
    componentLinearAccScale=<0.3 0.3 0.5>
    componentLinearVelScale=<0.2 0.2 0.3>
    overrideWorldGravity {z=-980}
   }
   Save cached pose 'WithBack_Pose'(Pose=Blend Poses by bool(
     BlendPose_0=Component To Local(ComponentPose=@n16)
     BlendPose_1=Component To Local(ComponentPose=@n22)
     bActiveValue=Get WearingCape() -> (WearingCape:bool)
    ) {blendType=Linear}) {cacheName=WithBack_Pose}
   n23 = Control Rig(Source=Copy Pose From Mesh() {bUseAttachedParent=true}) {
    alphaBoolBlend {blendOption=Linear}
    controlRigAssetReference {blueprintRigClass=&/F_MED_BrightIon_CtrlRig.F_MED_BrightIon_CtrlRig_C}
   }
   Save cached pose 'Start_Pose'(Pose=@n23) {cacheName=Start_Pose}
   n24 = "Transform (Modify) Bone - Bone: dyn_back_strap_01"!(ComponentPose=Local To Component(LocalPose=Use cached pose 'PreSim_Pose'()), Alpha:float=1, Rotation="0,0,0", Scale:Vector="1,1,1") -> (Pose) {
    rotationMode=BMM_Additive
    scaleMode=BMM_Additive
    scaleSpace=BCS_BoneSpace
    alphaBoolBlend {blendOption=Linear}
    boneToModify {boneName=dyn_back_strap_01}
   }
   n25 = "Constraint - dyn_back_strap_02 (P)"(ComponentPose=@n24, Alpha=1, ConstraintWeights_0=1) {
    alphaBoolBlend {blendOption=Linear}
    boneToModify {boneName=dyn_back_strap_02}
    constraintSetup [{targetBone {boneName=spine_02}}]
    constraintWeights [1]
   }
   n26 = "Transform (Modify) Bone - Bone: dyn_back_strap_02"(ComponentPose=@n25, Alpha=1, Translation="-0.000,2.152,0.000", Rotation="-0.000,0.000,7.203") {
    rotationMode=BMM_Additive
    scaleSpace=BCS_BoneSpace
    translationMode=BMM_Additive
    alphaBoolBlend {blendOption=Linear}
    boneToModify {boneName=dyn_back_strap_02}
    rotation {roll=7.203}
    translation {y=2.152}
   }
   n27 = "Transform (Modify) Bone - Bone: dyn_back_strap_02"!(ComponentPose=@n26, bAlphaBoolEnabled:bool=Get bIsInFrontEnd(), Translation="0,-2,0", Rotation="-4.617,0.000,-7.203") -> (Pose) {
    alphaInputType=Bool
    rotationMode=BMM_Additive
    scaleSpace=BCS_BoneSpace
    translationMode=BMM_Additive
    alphaBoolBlend {blendOption=Linear}
    boneToModify {boneName=dyn_back_strap_02}
    rotation {pitch=-4.617, roll=-7.203}
    translation {y=-2}
   }
   Save cached pose 'Back_Pose'(Pose=Component To Local(ComponentPose=@n27)) {cacheName=Back_Pose}
   n28 = "Transform (Modify) Bone - Bone: dyn_belt_l_MSTR"!(
    ComponentPose=Local To Component(LocalPose=Use cached pose 'Start_Pose'())
    Alpha=Get Alpha_Thigh_L_Front() -> (Alpha_Thigh_L_Front:double)
    Translation="0,0,9"
    Rotation="0,0,-90"
   ) -> (Pose) {
    rotationMode=BMM_Additive
    translationMode=BMM_Additive
    alphaBoolBlend {blendOption=Linear}
    boneToModify {boneName=dyn_belt_l_MSTR}
    rotation {roll=-90}
    translation {z=9}
   }
   n29 = "Transform (Modify) Bone - Bone: dyn_belt_r_MSTR"!(ComponentPose=@n28, Alpha=Get Alpha_Thigh_R_Front() -> (Alpha_Thigh_R_Front:double), Translation="0,0,5", Rotation="0,0,0") -> (Pose) {
    rotationMode=BMM_Additive
    translationMode=BMM_Additive
    alphaBoolBlend {blendOption=Linear}
    boneToModify {boneName=dyn_belt_r_MSTR}
    translation {z=5}
   }
   n30 = "Transform (Modify) Bone - Bone: attach_backpack"!(ComponentPose=@n29, bAlphaBoolEnabled:bool=Get bIsInFrontEnd(), Translation="0.632,2.320,-0.000", Rotation="0,0,0") -> (Pose) {
    alphaInputType=Bool
    translationMode=BMM_Additive
    translationSpace=BCS_BoneSpace
    alphaBoolBlend {blendOption=Linear}
    boneToModify {boneName=attach_backpack}
    translation=<0.632 2.32>
   }
   Save cached pose 'PreSim_Pose'(Pose=Component To Local(ComponentPose=@n30)) {cacheName=PreSim_Pose}
  ]}
 EventGraph {nodes [
   n1 = Event BlueprintUpdateAnimation()
   > n2 = Sequence!(execute) -> (then_0, then_1, then_2:Exec)
   n3 = Get SkelMesh()
   Set Alpha_Thigh_L_Front(
    execute=@n2.then_2
    Alpha_Thigh_L_Front:double=GetAlphaFromReferenceTransformDelta(SkeletalMeshComponent=@n3, BoneName=thigh_l, BaseName=pelvis, InputRange=(X=40,Y=-110), TwistAxis="1,0,0")
   ) -> (Output_Get:double)
   > Set Alpha_Thigh_R_Front(Alpha_Thigh_R_Front:double=GetAlphaFromReferenceTransformDelta(SkeletalMeshComponent=@n3, BoneName=thigh_r, BaseName=pelvis, InputRange=(X=40,Y=-110), TwistAxis="1,0,0")) -> (Output_Get:double)
   > Set Alpha_Necklace_L(Alpha_Necklace_L:double=GetAlphaFromReferenceTransformDelta(SkeletalMeshComponent=@n3, BoneName=dyn_necklace_02, BaseName=spine_05, DeltaComponentType=EulerY, InputRange=(X=0,Y=-20), TwistAxis="1,0,0")) -> (Output_Get:double)
   > Set Alpha_Necklace_R(Alpha_Necklace_R:double=GetAlphaFromReferenceTransformDelta(SkeletalMeshComponent=@n3, BoneName=dyn_necklace_02, BaseName=spine_05, DeltaComponentType=EulerY, InputRange=(X=0,Y=20), TwistAxis="1,0,0")) -> (Output_Get:double)
   n4 = Gravity Override State Function DEPRECATED(
    execute=@n2.then_1
    CharacterPartAnimInstance=Self-Reference()
    GameTimeInSeconds=GetGameTimeInSeconds()
    GravityOverrideParams=Get Gravity Override Params()
   )
   > Set Gravity Override(Gravity Override=@n4)
   n5 = SelectVector(A="0.2,0.2,0.2", B="0.35,0.35,0.35", bPickA=AND Boolean(A=Get bIsSkyDiving(), B=Get bIsFalling()))
   > Set ComponentLinearAccScale(ComponentLinearAccScale:Vector=@n5) -> (Output_Get:Vector)
   > Set ComponentLinearVelScale(ComponentLinearVelScale:Vector=@n5) -> (Output_Get:Vector)
   Is Valid(exec=@n2.then_0, InputObject=Get PlayerPawn() -> (PlayerPawn:Fort Player Pawn Athena Object Reference))
   > Set WPO Globals Function(
    Material=Get Mat Instance Dynamic()
    Overlay Material=Get Overlay Mat Instance Dynamic()
    Skeletal Mesh=Get SkelMesh()
    "Local Noise Transform  Location"="Get Local Noise Transform  Location"() -> ("Local Noise Transform  Location":Vector)
    Local Noise Offset=Get Local Noise Offset()
   )
   > Set WPO Regional Function!(
    execute
    Material:Material Instance Dynamic Object Reference=Get Mat Instance Dynamic()
    Overlay Material:Material Instance Dynamic Object Reference=Get Overlay Mat Instance Dynamic()
    Skeletal Mesh:Skeletal Mesh Component Object Reference=Get SkelMesh()
    Velocity Transform:Name=foot_r
    Velocity Interp Speed:double=1
    Collision Transform:Name=calf_r
    Collision Transform Axis:EAxis Enum
    Invert Collision Transform Axis:bool=false
    Delta Time X:double=@n1.DeltaTimeX
    Previous Velocity Transform Location:"Vector (by ref)"=Get R1 Velocity Transform Location()
    Previous Velocity Transform Velocity:"Vector (by ref)"=Get R1 Velocity Transform Velocity()
    Region:int=0
   ) -> (then)
   > Set WPO Regional Function(
    Material=Get Mat Instance Dynamic()
    Overlay Material=Get Overlay Mat Instance Dynamic()
    Skeletal Mesh=Get SkelMesh()
    Velocity Transform=foot_l
    Velocity Interp Speed=1
    Collision Transform=calf_l
    Invert Collision Transform Axis=true
    Delta Time X=@n1.DeltaTimeX
    Previous Velocity Transform Location=Get R2 Velocity Transform Location()
    Previous Velocity Transform Velocity=Get R2 Velocity Transform Velocity()
    Region=1
   )
   > n6 = Branch(Condition=Get bIsSkydiveFloating())
   > n7 = Set WPO Pho Wind Intensity(WPO Pho Wind Intensity=1)
   SetScalarParameterValue(execute=[@n8.then, @n7.then], self=Get Mat Instance Dynamic(), ParameterName=ClothSim_WindStrength, Value=Get WPO Pho Wind Intensity())
   n8 = Set WPO Pho Wind Intensity(execute=@n6.else, WPO Pho Wind Intensity=0)
   Event OnCharacterCustomizationLoaded()
   > n9 = Set SkelMesh(SkelMesh=GetPartSkeletalMeshComponent())
   > n10 = Get Dynamic Material Instance(SkeletalMeshComponent=@n9)
   > Set Mat Instance Dynamic(Mat Instance Dynamic=@n10)
   > n11 = GetOrCreateDynamicOverlayMaterial(MeshComponent=@n9, MaterialName=None)
   > Set Overlay Mat Instance Dynamic(Overlay Mat Instance Dynamic=@n11)
   n12 = CheckForBackbling()
   > Set WearingBack(WearingBack:bool=@n12) -> (Output_Get:bool)
   > Set WearingCape(WearingCape:bool=AND Boolean(A=@n12, B=@n12.bIsBackblingACape)) -> (Output_Get:bool)
   Event BlueprintInitializeAnimation()
   > Set SkelMesh(SkelMesh=GetOwnerSkeletalMeshComponent())
   n13 = TryGetPawnOwner()
   > Is Valid(InputObject=@n13)
   > n14 = Cast To FortPlayerPawnAthena(Object=@n13)
   > Set PlayerPawn(PlayerPawn:Fort Player Pawn Athena Object Reference=@n14) -> (Output_Get:Fort Player Pawn Athena Object Reference)
   Reroute Node(InputPin:Skeletal Mesh Component Object Reference=@n3) -> (OutputPin:Skeletal Mesh Component Object Reference)
  ]}
]
variables [
 Alpha_Necklace_L {type=number}
 Alpha_Necklace_R {type=number}
 Alpha_Thigh_L_Front {type=number}
 Alpha_Thigh_R_Front {type=number}
 ComponentLinearAccScale {type=Vector}
 ComponentLinearVelScale {type=Vector}
 Gravity Override {type=Vector}
 Gravity Override Params {type=GravityOverrideParamsStruct}
 Local Noise Offset {type=Vector}
 "Local Noise Transform  Location" {type=Vector}
 Mat Instance Dynamic {type=MaterialInstanceDynamic}
 Overlay Mat Instance Dynamic {type=MaterialInstanceDynamic}
 PlayerPawn {type=FortPlayerPawnAthena}
 R1 Velocity Transform Location {type=Vector}
 R1 Velocity Transform Velocity {type=Vector}
 R2 Velocity Transform Location {type=Vector}
 R2 Velocity Transform Velocity {type=Vector}
 SkelMesh {type=SkeletalMeshComponent}
 WearingBack {type=boolean}
 WearingCape {type=boolean}
 WPO Pho Wind Intensity {type=number}
]
